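; Parent Blueprint for a special searchable chest prop, handling its interaction check and the presentation of its searched state and arrow hint.
; Interaction is allowed whenever CanSearch is true.
; The IsSearched RepNotify stops the arrow effect, plays open sounds and particles, hides the mesh and smoke sheet, and fades out ambient audio.
; The ShowArrowHint RepNotify toggles the arrow particle's visibility and activation.
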(fn BlueprintCanInteract
  (params (InteractingPawn :type Fort%20Pawn%20Object%20Reference) (InteractionBeingAttempted :type EInteractionBeingAttempted%20Enum :default "FirstInteraction") (InteractionType :type TInteractionType%20Enum :default "IT_NoInteraction"))
  (return (Variables|Default|GetCanSearch?)))

(fn OnRep_IsSearched
  (if (Variables|Default|GetIsSearched)
    (Components|Activation|Deactivate (Variables|Default|GetP_ArrowInteraction))
    (Audio|Components|Audio|Play (Variables|Default|GetTiered_Chest_Open_T01_Cue) 0.0)
    (Audio|Components|Audio|Play (Variables|Default|GetTiered_Chest_Open_Cue) 0.0)
    (Components|Activation|Activate (Variables|Default|GetP_ChestOpen) false)
    (Development|SetHiddeninGame (Variables|Debug|GetStaticMeshComponent) true false)
    (Rendering|SetVisibility (Variables|Default|GetS_Chest_SmokeSheet) false false)
    (Audio|Components|Audio|FadeOut (Variables|Default|GetTiered_Chest_Ambient_Cue) 0.5 0.0 "Linear" :!class /Script/Engine.AudioComponent)))

(fn OnRep_ShowArrowHint
  (bind _p_arrowinteraction (Variables|Default|GetP_ArrowInteraction))
  (bind _p_arrowinteraction_2 (Variables|Default|GetP_ArrowInteraction))
  (if (Variables|Default|GetShowArrowHint)
    (Rendering|SetVisibility _p_arrowinteraction_2 true false)
    (Components|Activation|Activate _p_arrowinteraction_2 false)
    (else
      (Rendering|SetVisibility _p_arrowinteraction false false)
      (Components|Activation|Deactivate _p_arrowinteraction))))

(fn ConstructionScript)
